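# Animation Blueprint for the BrokenGravel pickaxe that gives its charm bones secondary motion.
# The AnimGraph passes the input pose through additive rotations on two charm bones, then two chained, angle-limited Anim Dynamics nodes with springs and reduced gravity in root-relative space.
# The update event sets charm rotator and alpha variables by front-end state, with rotators also checking whether it is off-pawn.
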AnimBlueprint /BRCosmetics/Weapons/FORT_Melee/Pickaxe_BrokenGravel/Meshes/BrokenGravel_Axe_AnimBP
target_skeleton=&/BrokenGravel_Axe_Skeleton
graphs [
 AnimGraph {nodes [
   AnimGraphNode_SubInput_0 = Input Pose()
   n1 = Local To Component(LocalPose=@AnimGraphNode_SubInput_0)
   n2 = Anim Dynamics - Bone: dyn_charm_01!(
    ComponentPose="Transform (Modify) Bone - Bone: dyn_charm_02"!(
     ComponentPose="Transform (Modify) Bone - Bone: base_charm_01"!(ComponentPose=@n1, Alpha=1, Rotation="-60,0,0") -> (Pose) {rotationMode=BMM_Additive, boneToModify {boneName=base_charm_01}, rotation {pitch=-60}}
     Alpha=1
     Rotation="-10,0,0"
    ) -> (Pose) {rotationMode=BMM_Additive, boneToModify {boneName=dyn_charm_02}, rotation {pitch=-10}}
    Alpha=0.7
    GravityOverride:Vector="0,0,-980"
   ) -> (Pose) {
    angularDampingOverride=1
    angularSpringConstant=9000
    bAngularSpring=true
    bOverrideAngularDamping=true
    bUseGravityOverride=true
    gravityScale=0.35
    simulationSpace=RootRelative
    boundBone {boneName=dyn_charm_01}
    componentAppliedLinearAccClamp=<3000 3000 3000>
    componentLinearAccScale=<0.4 0.4 0.4>
    componentLinearVelScale=<0.4 0.4 0.4>
    gravityOverride {z=-980}
    physicsBodyDefinitions [{
      boundBone {boneName=dyn_charm_01}
      boxExtents=<15 15 15>
      constraintSetup {angularLimitsMax=<0 90 40>, angularLimitsMin=<0 -20 -40>, angularTarget=<1 0 0>}
      localJointOffset=<4 0 0>
     }]
    simSpaceSettings {maxAngularAcceleration=20, maxAngularVelocity=10, simSpaceAngularAlpha=1}
   }
   n3 = Anim Dynamics - Bone: dyn_charm_02!(ComponentPose=@n2, Alpha=0.7, GravityOverride:Vector="0,0,-980") -> (Pose) {
    angularDampingOverride=1
    angularSpringConstant=9000
    bAngularSpring=true
    bOverrideAngularDamping=true
    bUseGravityOverride=true
    gravityScale=0.35
    simulationSpace=RootRelative
    boundBone {boneName=dyn_charm_02}
    componentAppliedLinearAccClamp=<3000 3000 3000>
    componentLinearAccScale=<0.4 0.4 0.4>
    componentLinearVelScale=<0.4 0.4 0.4>
    gravityOverride {z=-980}
    physicsBodyDefinitions [{
      boundBone {boneName=dyn_charm_02}
      boxExtents=<20 20 12>
      constraintSetup {angularLimitsMax=<0 90 40>, angularLimitsMin=<0 -60 -40>, angularTarget=<1 0 0>}
      localJointOffset=<6 0 0>
     }]
    simSpaceSettings {maxAngularAcceleration=20, maxAngularVelocity=10, simSpaceAngularAlpha=1}
   }
   Output Pose(Result=Component To Local(ComponentPose=@n3)) {layerGroup=None}
  ]}
 EventGraph {nodes [
   Event BlueprintUpdateAnimation()
   > n1 = Sequence()
   n2 = Get bIsOnPawn() -> (bIsOnPawn:bool)
   n3 = Get bIsInFrontEnd()
   > n4 = Branch(Condition=AND Boolean(A=NOT Boolean(A=@n2), B=@n3))
   > Set CharmBaseRotator(CharmBaseRotator:Rotator="-20,0,0") -> (Output_Get:Rotator)
   > Set CharmRotator(CharmRotator:Rotator="-4,0,0") -> (Output_Get:Rotator)
   Set CharmBaseRotator(execute=@n4.else, CharmBaseRotator="-60,0,0")
   > Set CharmRotator(CharmRotator="-10,0,0")
   n5 = Branch(execute=@n1.then_1, Condition=@n3)
   > Set AlphaFE(AlphaFE:double=0.3) -> (Output_Get:double)
   Set AlphaFE(execute=@n5.else, AlphaFE=0.7)
  ]}
]
variables [
 AlphaFE {type=number}
 CharmBaseRotator {type=Rotator}
 CharmRotator {type=Rotator}
]
